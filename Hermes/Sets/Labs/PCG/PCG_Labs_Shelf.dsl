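; PCG graph that fills shelves along a spline with book planes, using the Labs books material on instanced engine Plane meshes.
; Spline samples, set by Density, are stacked vertically, rotated upright, then given one of three Size/Scale pairs, weighted by the Ratio parameters.
; Margin-widened extents drive self-pruning of overlaps, then Coverage thins the rescaled survivors.
; The spawner packs Size into per-instance custom data.
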(graph PCG_Labs_Shelf
  (params (Density :type Int :default 32) (Coverage :type Double :default 1.0) (Margin :type Double :default 0.3) (Ratio_Large :type Double :default 1.2) (Ratio_Medium :type Double :default 1.5) (Ratio_Small :type Double :default 0.5))
  (bind GetSplineData_0 (Get%20Spline%20Data))
  (bind SplineSampler_0 (Spline%20Sampler :SamplerParams (struct :Mode NumberOfSamples :SubdivisionsPerSegment 8 :NumSamples 32 :StartOffset 5e+01 :EndOffset 5e+01 :PointSteepness 1.0) :NumSamples Density))
  (bind ExtentsModifier_0 (Extents%20Modifier :Extents (2e+01 2e+01 2e+01)))
  (bind DuplicatePoint_0 (Duplicate%20Point :Iterations 4 :PointTransform (struct :Translation (0.0 0.0 12.5))))
  (bind CreateConstant_0 (Create%20Constant :AttributeTypes (struct :DoubleValue 0.4)))
  (bind CreateConstant_1 (Create%20Constant :AttributeTypes (struct :Type Float :FloatValue 4.0 :Int32Value 4) :OutputTarget Size))
  (bind CreateConstant_2 (Create%20Constant :AttributeTypes (struct :Type Float :FloatValue 8.0 :Int32Value 1) :OutputTarget Size))
  (bind CreateConstant_3 (Create%20Constant :AttributeTypes (struct :Type Float :FloatValue 16.0 :Int32Value 2) :OutputTarget Size))
  (bind MergeAttributes_0 (Merge%20Attributes))
  (bind CreateConstant_5 (Create%20Constant :AttributeTypes (struct :Type Float :FloatValue 2.0 :Int32Value 4) :OutputTarget Scale))
  (bind CreateConstant_6 (Create%20Constant :AttributeTypes (struct :Type Float :FloatValue 1.0 :Int32Value 4) :OutputTarget Scale))
  (bind CreateConstant_7 (Create%20Constant :AttributeTypes (struct :Type Float :FloatValue 0.5 :Int32Value 4) :OutputTarget Scale))
  (bind MergeAttributes_1 (Merge%20Attributes))
  (bind AttributeRename_0 (Attribute%20Rename :AttributeToRename @LastCreated :NewAttributeName Weight :In Ratio_Large))
  (bind AttributeRename_1 (Attribute%20Rename :AttributeToRename @LastCreated :NewAttributeName Weight :In Ratio_Medium))
  (bind AttributeRename_2 (Attribute%20Rename :AttributeToRename @LastCreated :NewAttributeName Weight :In Ratio_Small))
  (bind MergeAttributes_4 (Merge%20Attributes))
  (bind AddAttribute_0 (Add%20Attribute :bCopyAllAttributes true))
  (bind TransformPoints_0 (Transform%20Points :OffsetMax (0.0 0.0 22.0) :OffsetMin (0.0 0.0 22.0) :RotationMax (struct :Roll 9e+01) :RotationMin (struct :Roll 9e+01)))
  (bind MatchAndSetAttributes_0 (Match%20And%20Set%20Attributes :bUseWeightAttribute true :WeightAttribute Weight))
  (bind AttributeMathsOp_2 (Attribute%20Maths%20Op :InputSource1 $Scale.X :InputSource2 Scale :Operation Multiply))
  (bind CreateConstant_10 (Create%20Constant :AttributeTypes (struct :DoubleValue 1.0)))
  (bind AttributeMathsOp_3 (Attribute%20Maths%20Op :InA Margin))
  (bind CreateConstant_8 (Create%20Constant :AttributeTypes (struct :DoubleValue 1.0)))
  (bind MakeVectorAttribute_0 (Make%20Vector%20Attribute :OutputType Vector))
  (bind ExtentsModifier_1 (Extents%20Modifier :Mode Multiply))
  (bind SelfPruning_0 (Self%20Pruning :Parameters (struct :PruningType AllEqual :ComparisonSource $Extents)))
  (bind TransformPoints_1 (Transform%20Points))
  (bind SelectPoints_0 (Select%20Points :Ratio-param Coverage))
  (bind StaticMeshSpawner_0 (Static%20Mesh%20Spawner :bApplyMeshBoundsToPoints false :InstanceDataPackerParameters (class:/Script/PCG.PCGInstanceDataPackerByAttribute :AttributeSelectors [Size]) :InstanceDataPackerType /Script/PCG.PCGInstanceDataPackerByAttribute :MeshSelectorParameters (class:/Script/PCG.PCGMeshSelectorWeighted :MeshEntries [(struct :Descriptor (struct :StaticMesh /Engine/BasicShapes/Plane.Plane :OverrideMaterials [/Hermes/Sets/Labs/Textures/BookShelves/MI_Labs_Books.MI_Labs_Books] :bCastShadow false :bAffectDistanceFieldLighting false :bCastContactShadow false :bIncludeInHLOD false))])))
  (connect MergeAttributes_1 Out AddAttribute_0 Attributes)
  (connect MergeAttributes_4 Out AddAttribute_0 Attributes)
  (connect MergeAttributes_0 Out AddAttribute_0 In)
  (connect MatchAndSetAttributes_0 Out AttributeMathsOp_2 InA)
  (connect MatchAndSetAttributes_0 Out AttributeMathsOp_2 InB)
  (connect CreateConstant_10 Out AttributeMathsOp_3 InB)
  (connect ExtentsModifier_0 Out DuplicatePoint_0 In)
  (connect SplineSampler_0 Out ExtentsModifier_0 In)
  (connect MakeVectorAttribute_0 Out ExtentsModifier_1 Extents)
  (connect AttributeMathsOp_2 Out ExtentsModifier_1 In)
  (connect AttributeMathsOp_3 Out MakeVectorAttribute_0 X)
  (connect CreateConstant_8 Out MakeVectorAttribute_0 Y)
  (connect CreateConstant_8 Out MakeVectorAttribute_0 Z)
  (connect TransformPoints_0 Out MatchAndSetAttributes_0 In)
  (connect AddAttribute_0 Out MatchAndSetAttributes_0 Match%20Data)
  (connect CreateConstant_1 Out MergeAttributes_0 In)
  (connect CreateConstant_2 Out MergeAttributes_0 In)
  (connect CreateConstant_3 Out MergeAttributes_0 In)
  (connect CreateConstant_5 Out MergeAttributes_1 In)
  (connect CreateConstant_6 Out MergeAttributes_1 In)
  (connect CreateConstant_7 Out MergeAttributes_1 In)
  (connect AttributeRename_0 Out MergeAttributes_4 In)
  (connect AttributeRename_1 Out MergeAttributes_4 In)
  (connect AttributeRename_2 Out MergeAttributes_4 In)
  (connect TransformPoints_1 Out SelectPoints_0 In)
  (connect ExtentsModifier_1 Out SelfPruning_0 In)
  (connect GetSplineData_0 Out SplineSampler_0 Spline)
  (connect SelectPoints_0 Out StaticMeshSpawner_0 In)
  (connect DuplicatePoint_0 Out TransformPoints_0 In)
  (connect SelfPruning_0 Out TransformPoints_1 In)
  (connect CreateConstant_0 Out TransformPoints_1 ScaleMax)
  (connect CreateConstant_0 Out TransformPoints_1 ScaleMin))
